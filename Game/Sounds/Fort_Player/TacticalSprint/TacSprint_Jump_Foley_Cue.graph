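# Sound cue for tactical-sprint jump foley, layering gear rustle, boot scuff, sprint footstep and whooshes.
# Islocal picks first-person attenuation and sound class, otherwise third-person, where IsEnemy keeps enemies louder and makes others nearly silent.
# All routes share one mixer, including a delayed sub-mix that staggers weapon-swap rustles with pitched-down Shinobi emote whooshes.
# Quality-level nodes switch jump and scuff layers between larger and smaller random wave pools.
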SoundCue /Game/Sounds/Fort_Player/TacticalSprint/TacSprint_Jump_Foley_Cue
root=Branch_1
nodes [
 {
  id=Attenuation_0
  children [SoundClass_0]
  properties {attenuationSettings=/Game/Sounds/Attenuation/Fort_Player_Footsteps_3P_Sprint_Athena_Att}
 }
 {
  id=Attenuation_1
  children [SoundClass_1]
  properties {attenuationSettings=/Game/Sounds/Attenuation/Fort_Player_Footsteps_1P_Sprint_Athena_Att}
 }
 {id=Branch_0, children [Modulator_1,Modulator_0,Modulator_0], properties {boolParameterName=IsEnemy}}
 {id=Branch_1, children [Attenuation_1, Branch_0, Branch_0], properties {boolParameterName=Islocal}}
 {id=Delay_0, children [Random_8], properties {delayMax=0.24, delayMin=0.22}}
 {id=Delay_1, children [Random_5], properties {delayMax=0.47, delayMin=0.45}}
 {id=Delay_2, children [Mixer_1], properties {delayMax=0.22, delayMin=0.19}}
 {id=Delay_3, children [Modulator_3], properties {delayMax=0.29, delayMin=0.25}}
 {id=Delay_4, children [Modulator_3], properties {delayMax=0.54, delayMin=0.5}}
 {id=Delay_5, children [QualityLevel_5], properties {delayMax=0.08, delayMin=0.08}}
 {id=Mixer_1, children [Random_9,Modulator_3,Delay_0,Delay_3,Delay_1,Delay_4], properties {inputVolume [0.8,1,0.8,1,0.8,1]}}
 {id=Mixer_3, children [QualityLevel_4, Delay_5, Modulator_4, Modulator_2], properties {inputVolume [1.2,1.4,2,0.8]}}
 {id=Modulator_0, children [Attenuation_0], properties {pitchMax=1, pitchMin=1, volumeMax=0.03, volumeMin=0.03}}
 {id=Modulator_1, children [Attenuation_0], properties {pitchMax=1, pitchMin=1, volumeMax=0.5, volumeMin=0.5}}
 {id=Modulator_2, children [Delay_2], properties {pitchMax=1.5, pitchMin=1.3, volumeMax=1, volumeMin=0.55}}
 {id=Modulator_3, children [Random_10], properties {pitchMax=0.8, pitchMin=0.5, volumeMax=0.35, volumeMin=0.2}}
 {id=Modulator_4, children [Random_1], properties {pitchMax=1.35, pitchMin=1.25, volumeMax=0.9, volumeMin=0.85}}
 {id=QualityLevel_4, children [Random_12,Random_12,Random_13,Random_13]}
 {id=QualityLevel_5, children [Random_14,Random_14,Random_15,Random_15]}
 {id=Random_1, children [WavePlayer_0, WavePlayer_1, WavePlayer_8, WavePlayer_15], properties {weights [1,1,1,1]}}
 {id=Random_10, children [WavePlayer_23, WavePlayer_25], properties {bShouldExcludeFromBranchCulling=true, weights [1,1]}}
 {
  id=Random_12
  children [WavePlayer_60, WavePlayer_61, WavePlayer_62, WavePlayer_63, WavePlayer_64, WavePlayer_65, WavePlayer_66, WavePlayer_67]
  properties {weights [1,1,1,1,1,1,1,1]}
 }
 {
  id=Random_13
  children [WavePlayer_70, WavePlayer_69, WavePlayer_68, WavePlayer_83, WavePlayer_82]
  properties {bShouldExcludeFromBranchCulling=true, weights [1,1,1,1,1]}
 }
 {
  id=Random_14
  children [WavePlayer_71, WavePlayer_72, WavePlayer_73, WavePlayer_74, WavePlayer_76, WavePlayer_77]
  properties {weights [1,1,1,1,1,1]}
 }
 {
  id=Random_15
  children [WavePlayer_78, WavePlayer_79, WavePlayer_80, WavePlayer_81]
  properties {bShouldExcludeFromBranchCulling=true, weights [1,1,1,1]}
 }
 {id=Random_5, children [WavePlayer_84, WavePlayer_13, WavePlayer_85], properties {bShouldExcludeFromBranchCulling=true, weights [1,1,1]}}
 {id=Random_8, children [WavePlayer_9, WavePlayer_10, WavePlayer_11], properties {bShouldExcludeFromBranchCulling=true, weights [1,1,1]}}
 {
  id=Random_9
  children [WavePlayer_3,WavePlayer_5,WavePlayer_6,WavePlayer_7]
  properties {bShouldExcludeFromBranchCulling=true, weights [1,1,1,1]}
 }
 {id=SoundClass_0, children [Mixer_3], properties {soundClassOverride=&Fort_Player_Foley_3P_Enemy_Athena}}
 {id=SoundClass_1, children [Mixer_3], properties {soundClassOverride=&Fort_Player_Foley_1P_Athena}}
 {
  id=WavePlayer_0
  properties {soundWaveAssetPtr=/Game/Sounds/Fort_Player/Footsteps/CombatBoots/Waves/1P/Asphalt/Foley_FS_1P_BootsCombat_Asphalt_Sprint_01}
 }
 {
  id=WavePlayer_1
  properties {soundWaveAssetPtr=/Game/Sounds/Fort_Player/Footsteps/CombatBoots/Waves/1P/Asphalt/Foley_FS_1P_BootsCombat_Asphalt_Sprint_02}
 }
 {
  id=WavePlayer_10
  properties {soundWaveAssetPtr=/Game/Sounds/Fort_Player/Clothing/TacticalGear/Waves/Foley_Clothing_TacticalGear_SwapWeapon_08}
 }
 {
  id=WavePlayer_11
  properties {soundWaveAssetPtr=/Game/Sounds/Fort_Player/Clothing/TacticalGear/Waves/Foley_Clothing_TacticalGear_SwapWeapon_09}
 }
 {
  id=WavePlayer_13
  properties {soundWaveAssetPtr=/Game/Sounds/Fort_Player/Clothing/TacticalGear/Waves/Foley_Clothing_TacticalGear_SwapWeapon_06}
 }
 {
  id=WavePlayer_15
  properties {soundWaveAssetPtr=/Game/Sounds/Fort_Player/Footsteps/CombatBoots/Waves/1P/Asphalt/Foley_FS_1P_BootsCombat_Asphalt_Sprint_08}
 }
 {id=WavePlayer_23, properties {soundWaveAssetPtr=/Game/Athena/Sounds/Emotes/Shinobi/Whoosh/Emote_Shinobi_Whoosh_01}}
 {id=WavePlayer_25, properties {soundWaveAssetPtr=/Game/Athena/Sounds/Emotes/Shinobi/Whoosh/Emote_Shinobi_Whoosh_02}}
 {
  id=WavePlayer_3
  properties {soundWaveAssetPtr=/Game/Sounds/Fort_Player/Clothing/TacticalGear/Waves/Foley_Clothing_TacticalGear_SwapWeapon_01}
 }
 {
  id=WavePlayer_5
  properties {soundWaveAssetPtr=/Game/Sounds/Fort_Player/Clothing/TacticalGear/Waves/Foley_Clothing_TacticalGear_SwapWeapon_03}
 }
 {
  id=WavePlayer_6
  properties {soundWaveAssetPtr=/Game/Sounds/Fort_Player/Clothing/TacticalGear/Waves/Foley_Clothing_TacticalGear_SwapWeapon_04}
 }
 {id=WavePlayer_60, properties {soundWaveAssetPtr=/Game/Sounds/Fort_Player/Clothing/TacticalGear/Waves/Foley_Clothing_TacticalGear_Jump_01}}
 {id=WavePlayer_61, properties {soundWaveAssetPtr=/Game/Sounds/Fort_Player/Clothing/TacticalGear/Waves/Foley_Clothing_TacticalGear_Jump_02}}
 {id=WavePlayer_62, properties {soundWaveAssetPtr=/Game/Sounds/Fort_Player/Clothing/TacticalGear/Waves/Foley_Clothing_TacticalGear_Jump_03}}
 {id=WavePlayer_63, properties {soundWaveAssetPtr=/Game/Sounds/Fort_Player/Clothing/TacticalGear/Waves/Foley_Clothing_TacticalGear_Jump_04}}
 {id=WavePlayer_64, properties {soundWaveAssetPtr=/Game/Sounds/Fort_Player/Clothing/TacticalGear/Waves/Foley_Clothing_TacticalGear_Jump_05}}
 {id=WavePlayer_65, properties {soundWaveAssetPtr=/Game/Sounds/Fort_Player/Clothing/TacticalGear/Waves/Foley_Clothing_TacticalGear_Jump_06}}
 {id=WavePlayer_66, properties {soundWaveAssetPtr=/Game/Sounds/Fort_Player/Clothing/TacticalGear/Waves/Foley_Clothing_TacticalGear_Jump_08}}
 {id=WavePlayer_67, properties {soundWaveAssetPtr=/Game/Sounds/Fort_Player/Clothing/TacticalGear/Waves/Foley_Clothing_TacticalGear_Jump_10}}
 {id=WavePlayer_68, properties {soundWaveAssetPtr=&Foley_Clothing_TacticalGear_Jump_05}}
 {id=WavePlayer_69, properties {soundWaveAssetPtr=&Foley_Clothing_TacticalGear_Jump_03}}
 {
  id=WavePlayer_7
  properties {soundWaveAssetPtr=/Game/Sounds/Fort_Player/Clothing/TacticalGear/Waves/Foley_Clothing_TacticalGear_SwapWeapon_05}
 }
 {id=WavePlayer_70, properties {soundWaveAssetPtr=&Foley_Clothing_TacticalGear_Jump_01}}
 {
  id=WavePlayer_71
  properties {soundWaveAssetPtr=/Game/Sounds/Fort_Player/Footsteps/CombatBoots/Waves/1P/Asphalt/Foley_FS_1P_BootsCombat_Asphalt_Scuff_01}
 }
 {
  id=WavePlayer_72
  properties {soundWaveAssetPtr=/Game/Sounds/Fort_Player/Footsteps/CombatBoots/Waves/1P/Asphalt/Foley_FS_1P_BootsCombat_Asphalt_Scuff_02}
 }
 {
  id=WavePlayer_73
  properties {soundWaveAssetPtr=/Game/Sounds/Fort_Player/Footsteps/CombatBoots/Waves/1P/Asphalt/Foley_FS_1P_BootsCombat_Asphalt_Scuff_03}
 }
 {
  id=WavePlayer_74
  properties {soundWaveAssetPtr=/Game/Sounds/Fort_Player/Footsteps/CombatBoots/Waves/1P/Asphalt/Foley_FS_1P_BootsCombat_Asphalt_Scuff_04}
 }
 {
  id=WavePlayer_76
  properties {soundWaveAssetPtr=/Game/Sounds/Fort_Player/Footsteps/CombatBoots/Waves/1P/Asphalt/Foley_FS_1P_BootsCombat_Asphalt_Scuff_06}
 }
 {
  id=WavePlayer_77
  properties {soundWaveAssetPtr=/Game/Sounds/Fort_Player/Footsteps/CombatBoots/Waves/1P/Asphalt/Foley_FS_1P_BootsCombat_Asphalt_Scuff_07}
 }
 {
  id=WavePlayer_78
  properties {soundWaveAssetPtr=&Foley_FS_1P_BootsCombat_Asphalt_Scuff_01}
 }
 {
  id=WavePlayer_79
  properties {soundWaveAssetPtr=&Foley_FS_1P_BootsCombat_Asphalt_Scuff_02}
 }
 {
  id=WavePlayer_8
  properties {soundWaveAssetPtr=/Game/Sounds/Fort_Player/Footsteps/CombatBoots/Waves/1P/Asphalt/Foley_FS_1P_BootsCombat_Asphalt_Sprint_05}
 }
 {
  id=WavePlayer_80
  properties {soundWaveAssetPtr=&Foley_FS_1P_BootsCombat_Asphalt_Scuff_03}
 }
 {
  id=WavePlayer_81
  properties {soundWaveAssetPtr=&Foley_FS_1P_BootsCombat_Asphalt_Scuff_04}
 }
 {id=WavePlayer_82, properties {soundWaveAssetPtr=&Foley_Clothing_TacticalGear_Jump_10}}
 {id=WavePlayer_83, properties {soundWaveAssetPtr=&Foley_Clothing_TacticalGear_Jump_08}}
 {
  id=WavePlayer_84
  properties {soundWaveAssetPtr=&Foley_Clothing_TacticalGear_SwapWeapon_03}
 }
 {
  id=WavePlayer_85
  properties {soundWaveAssetPtr=&Foley_Clothing_TacticalGear_SwapWeapon_04}
 }
 {
  id=WavePlayer_9
  properties {soundWaveAssetPtr=/Game/Sounds/Fort_Player/Clothing/TacticalGear/Waves/Foley_Clothing_TacticalGear_SwapWeapon_07}
 }
]
properties {
 attenuationSettings=/Game/Sounds/Attenuation/Fort_Player_Footsteps_Att
 maxDistance=6000
 soundClassObject=&Fort_Player_Foley_1P_Athena
 volumeMultiplier=2
}
