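# Sound cue for the player's tactical sprint jump whoosh, routed to the player foley sound class.
# A random node picks one of six equally weighted whoosh waves.
# A modulator at the root slightly raises and varies the pitch on each play.
# Spatialization uses the shared basic foley attenuation settings.
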SoundCue /Game/Sounds/Fort_Player/TacticalSprint/Tac_Sprint_Jump_Whoosh
root=Modulator_0
nodes [
 {id=Modulator_0, children [Random_1], properties {pitchMax=1.15, pitchMin=1.05}}
 {id=Random_1, children [WavePlayer_3,WavePlayer_4,WavePlayer_5,WavePlayer_6,WavePlayer_7,WavePlayer_8], properties {weights [1,1,1,1,1,1]}}
 {id=WavePlayer_3, properties {soundWaveAssetPtr=&/Tac_Sprint_Jump_Whoosh_01}}
 {id=WavePlayer_4, properties {soundWaveAssetPtr=&/Tac_Sprint_Jump_Whoosh_02}}
 {id=WavePlayer_5, properties {soundWaveAssetPtr=&/Tac_Sprint_Jump_Whoosh_03}}
 {id=WavePlayer_6, properties {soundWaveAssetPtr=&/Tac_Sprint_Jump_Whoosh_04}}
 {id=WavePlayer_7, properties {soundWaveAssetPtr=&/Tac_Sprint_Jump_Whoosh_05}}
 {id=WavePlayer_8, properties {soundWaveAssetPtr=&/Tac_Sprint_Jump_Whoosh_06}}
]
properties {
 attenuationSettings=&Foley_Attenuation_Basic_Att
 maxDistance=1536
 soundClassObject=&Fort_Player_Foley
 volumeMultiplier=1
}
